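# Idle Niagara effect for the BikeMold backpack cosmetic: blue thruster jets and glow attached to skeletal mesh sockets.
# Jets places four mesh-and-sprite particles on separate sockets, with random scale flicker and waveform-pulsed decal color.
# Two system timelines stretch jets for jumping and wing folding, taking the larger scale.
# GlowDecalAndLight adds a fading-in light and socket-oriented capsule glow, using a front-end/gameplay modifier for capsule alpha.
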NiagaraSystem /BRCosmetics/Accessories/FORT_Backpacks/Backpack_M_MED_BikeMold/FX/NS_Backpack_BikeMold_Idle
dependencies {
 usedDataInterfaces [&NiagaraDataInterfaceCurve, &NiagaraDataInterfaceSkeletalMesh, &NiagaraDataInterfaceArrayFloat]
 usedDynamicInputs [&VectorLength]
 usedModules [
  &self:SystemSpawnScript.SystemScriptSource.SystemScriptGraph.NiagaraNodeAssignment_0.SetVariables_#1
  &SystemState
  &Timeline
  &self:SystemSpawnScript.SystemScriptSource.SystemScriptGraph.NiagaraNodeAssignment_1.SetVariables_#2
 ]
}
emitters {
 GlowDecalAndLight {
  emitterSpawnScript [/SetVariables(Emitter.CapsuleGlowDirection:Vector3f=<0 1 0>) {isSetParameters=true}]
  emitterUpdateScript [
   &EmitterState()
   &SpawnBurst_Instantaneous(Use Spawn Probability=false, UseLoopCountLimit=true) {module=SpawnBurst_Instantaneous001}
  ]
  particleSpawnScript [
   &InitializeParticle(
    Color=<rgba 0 0.2917 0.5 1>
    Mesh Scale=<0.025 0.001 0.001>
    Mesh Scale Mode=::NewEnumerator4
    Sprite Size Mode=::NewEnumerator3
    Uniform Sprite Size=5
    UsePositionOffset=true
    Write Material Random=true
   )
   /SetVariables(Particles.CapsuleCameraOffset=1, Particles.CapsuleSize:Vector2f=&Vector2DFromFloat(Value=12)) {isSetParameters=true}
   &SkeletalMeshLocation(
    Bone / Socket Index=5
    Interpolate Mesh Sampling=false
    MaskSpawnGroup=false
    Mesh Sampling Type=::NewEnumerator5
    Skeletal Mesh=$System.Skeletal Mesh
    Socket Sampling Mode=::NewEnumerator5
   )
  ]
  particleUpdateScript [
   &ParticleState(Kill Particles When Lifetime Has Elapsed=false)
   &ScaleSpriteSize(Uniform Curve Sprite Scale=DataInterface() {propertyValues {Curve {keys [
        {arriveTangent=1.1179, interpMode=RCIM_Cubic, leaveTangent=1.1179, tangentMode=RCTM_User, time=0, value=0.6}
        {arriveTangent=0.0224, interpMode=RCIM_Cubic, leaveTangent=0.0224, tangentMode=RCTM_User, time=1, value=1}
       ]}}})
   &CameraOffset()
   &ScaleColor(
    Scale Alpha=&FloatFromCurve(FloatCurve=DataInterface() {propertyValues {Curve {keys [
         {arriveTangent=4.1785, interpMode=RCIM_Cubic, leaveTangent=4.1785, tangentMode=RCTM_User, time=0, value=0}
         {arriveTangent=1.5009, interpMode=RCIM_Cubic, leaveTangent=1.5009, tangentMode=RCTM_User, time=0.2, value=0.55}
         {arriveTangent=0.0224, interpMode=RCIM_Cubic, leaveTangent=0.0224, tangentMode=RCTM_User, time=1, value=1}
        ]}}})
    Scale RGB=&VectorFromFloat(Value=&FloatFromCurve(
      FloatCurve=DataInterface() {propertyValues {Curve {keys [
          {arriveTangent=4.1785, interpMode=RCIM_Cubic, leaveTangent=4.1785, tangentMode=RCTM_User, time=0, value=0}
          {arriveTangent=1.5009, interpMode=RCIM_Cubic, leaveTangent=1.5009, tangentMode=RCTM_User, time=0.2, value=0.55}
          {arriveTangent=0.0224, interpMode=RCIM_Cubic, leaveTangent=0.0224, tangentMode=RCTM_User, time=1, value=1}
         ]}}}
      Scale Curve=5
     ))
    ScaleA=true
    ScaleRGB=true
   )
   /SetVariables(
    Particles.CapsuleColor:LinearColor=&MultiplyLinearColor(
     Linear Color=$Particles.Color
     Scale Factor=&MakeVector4(X=1, Y=1, Z=1, W=&FN_FrontEndFloatModifier(Gameplay=0.6, Front End=0.2))
    )
    Particles.CapsulePosition:NiagaraPosition=&AddVectorToPosition(
     Position=$Particles.Position
     Vector=&RotateVectorByQuaternion(Vector To Rotate=<0 -1 0>, Quaternion=$Particles.SkeletalMeshLocation.SampledRotation)
    )
   ) {isSetParameters=true}
   &DynamicMaterialParameters(
    Index 0 Param 1=&MakeFloatFromVector(
     Vector=&RotateVectorByQuaternion(Vector To Rotate=$Emitter.CapsuleGlowDirection, Quaternion=$Particles.SkeletalMeshLocation.SampledRotation)
     Channel=::NewEnumerator0
    )
    Index 0 Param 2=&MakeFloatFromVector(
     Vector=&RotateVectorByQuaternion(Vector To Rotate=$Emitter.CapsuleGlowDirection, Quaternion=$Particles.SkeletalMeshLocation.SampledRotation)
     Channel=::NewEnumerator1
    )
    Index 0 Param 3=&MakeFloatFromVector(
     Vector=&RotateVectorByQuaternion(Vector To Rotate=$Emitter.CapsuleGlowDirection, Quaternion=$Particles.SkeletalMeshLocation.SampledRotation)
     Channel=::NewEnumerator2
    )
    Index 0 Param 4=10
    Module.Param0WriteEnabled=true
    Module.Param1WriteEnabled=true
    Module.Param2WriteEnabled=true
    Module.Param3WriteEnabled=true
   )
   &FN_Light_Attributes(
    Color=$Particles.Color
    Create Color Attribute=true
    Create Radius Attribute=true
    Radius=12
    Scale Color=&MakeFloatFromVector4(
     Vector 4=&MakeVector4_FromFloat(Float=&FloatFromCurve(FloatCurve=DataInterface() {propertyValues {Curve {keys [
           {arriveTangent=4.1785, interpMode=RCIM_Cubic, leaveTangent=4.1785, tangentMode=RCTM_User, time=0, value=0}
           {arriveTangent=1.5009, interpMode=RCIM_Cubic, leaveTangent=1.5009, tangentMode=RCTM_User, time=0.2, value=0.55}
           {arriveTangent=0.0224, interpMode=RCIM_Cubic, leaveTangent=0.0224, tangentMode=RCTM_User, time=1, value=1}
          ]}}}))
     Channel=::NewEnumerator0
    )
   )
  ]
  rendererClasses [&NiagaraLightRendererProperties, &NiagaraSpriteRendererProperties]
  renderers [
   {rendererClass=&NiagaraLightRendererProperties, rendererIndex=0}
   {rendererClass=&NiagaraSpriteRendererProperties, rendererIndex=1}
   {rendererClass=&NiagaraSpriteRendererProperties, rendererIndex=2}
  ]
 }
 Jets {
  emitterUpdateScript [
   &EmitterState()
   &SpawnBurst_Instantaneous(Spawn Count=4, Use Spawn Probability=false, UseLoopCountLimit=true)
  ]
  particleSpawnScript [
   &PartitionParticles(Number Of Partitions=4, Particles Per Partition=1, Partition Mode:ENiagara_PartitionMode=::NewEnumerator1)
   &InitializeParticle(
    Color=<rgba 0 0.5833 1 1>
    Mesh Scale=<0.03 0.001 0.001>
    Mesh Scale Mode=::NewEnumerator4
    Sprite Size Mode=::NewEnumerator3
    UsePositionOffset=false
    Write Material Random=true
   )
   &DynamicMaterialParameters(
    Index 0 Param 1=&RandomRangeFloat(Minimum=0, Maximum=1)
    Module.Param0WriteEnabled=true
    Module.Param1WriteEnabled=true
    Module.Param2WriteEnabled=true
    Module.Param3WriteEnabled=true
   )
   &CameraOffset(Camera Offset Amount=-0.5) {enabled=false}
   /SetVariables(
    Particles.ColorAirGlow:LinearColor=<rgba 0 1.75 3 0.02>
    Particles.ColorFlame:LinearColor=<rgba 0.0169 1.1737 2 1>
    Particles.ColorGlow:LinearColor=<rgba 0 1.75 3 0.05>
   ) {isSetParameters=true}
   /SetVariables(Particles.CamOffsetAirGlow=2, Particles.CamOffsetDecal=1, Particles.CamOffsetFlame=1) {isSetParameters=true}
   /SetVariables(
    Particles.DynamicMaterialAirGlow:Vector4f=&MakeVector4(X=0, Y=1, Z=0, W=0)
    Particles.DynamicMaterialDecal:Vector4f=&MakeVector4(X=2, Y=0, Z=0, W=0)
   ) {isSetParameters=true}
  ]
  particleUpdateScript [
   &ParticleState(Kill Particles When Lifetime Has Elapsed=false, Let Infinitely Lived Particles Die When Emitter Deactivates=true)
   &SkeletalMeshLocation(
    Bone / Socket Index=$Particles.PartitionParticles.Partition
    MaskSpawnGroup=false
    Mesh Sampling Type=::NewEnumerator5
    Skeletal Mesh=$System.Skeletal Mesh
    Socket Sampling Mode=::NewEnumerator5
   )
   /SetVariables(Particles.ScaleByJump:Vector3f=&MakeVector(X=&Lerp_Float(A=1, B=2, Alpha=$System.Timeline.NormalizedTime), Y=1, Z=1)) {isSetParameters=true}
   /SetVariables(Particles.ScaleByFoldingWings:Vector3f=&MakeVector(
     X=&Lerp_Float(A=1, B=2, Alpha=$System.Timeline001.NormalizedTime)
     Y=&Lerp_Float(A=1, B=1.5, Alpha=$System.Timeline001.NormalizedTime)
     Z=&Lerp_Float(A=1, B=1.5, Alpha=$System.Timeline001.NormalizedTime)
    )) {isSetParameters=true}
   &ScaleMeshSize(Scale Factor=/NiagaraFluids/DynamicInput/MaxVectorByLength(A:Vector3f=$Particles.ScaleByJump, B:Vector3f=$Particles.ScaleByFoldingWings)) {module=ScaleMeshSize003}
   &ScaleMeshSize(Scale Factor=&MakeCustomVectorFromBool(
     Bool=&SetBoolByFloatComparison(
      A=&MakeFloatFromVector(Vector=&Add_Vector(A=$Particles.ScaleByJump, B=$Particles.ScaleByFoldingWings), Channel=::NewEnumerator0)
      B=4.1
     )
     True Vector=<0.8 1 1>
     False Vector=<1 1 1>
    )) {module=ScaleMeshSize002}
   &ScaleMeshSize(Scale Factor=&RandomRangeVector(Minimum=<0.8 0.8 0.8>, Maximum=<1.25 1 1>, Evaluation Type=::NewEnumerator1))
   /SetVariables(Particles.ColorDecal:LinearColor=&MultiplyLinearColor(
     Linear Color=<rgba 0 0.5833 1 0.5>
     Scale Factor=&MakeVector4_FromFloat(Float=&Waveform(
       Global Input Position=&ReturnNormalizedExecIndex()
       "Waveform [1]"=::NewEnumerator0
       "[1] Frequency"=500
       "[1] Amplitude Min/Max"=<0.5 1>
       "[1] Bias"=0.5
      ))
    )) {isSetParameters=true}
  ]
  rendererClasses [&NiagaraMeshRendererProperties, &NiagaraSpriteRendererProperties]
  renderers [
   {rendererClass=&NiagaraMeshRendererProperties, rendererIndex=0}
   {rendererClass=&NiagaraMeshRendererProperties, rendererIndex=1}
   {rendererClass=&NiagaraSpriteRendererProperties, rendererIndex=2}
   {rendererClass=&NiagaraSpriteRendererProperties, rendererIndex=3}
  ]
 }
}
summary {
 systemName=NS_Backpack_BikeMold_Idle
 emitters [
  {emitterName=Jets, rendererClasses [&NiagaraMeshRendererProperties, &NiagaraSpriteRendererProperties]}
  {
   emitterName=GlowDecalAndLight
   rendererClasses [&NiagaraLightRendererProperties, &NiagaraSpriteRendererProperties]
  }
 ]
 userVariables [
  User.bIsFoldingWings {type=NiagaraBool}
  User.bIsFrontEnd {type=NiagaraBool}
  User.bIsFrontEndPreview {type=NiagaraBool}
  User.bPawnJumped {type=NiagaraBool}
 ]
}
