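# Anim Blueprint for the BikeMold backpack that poses and jiggles its four wing-like bones as the player moves.
# EventGraph derives flags and alphas from locomotion, skydive and spine-twist state, and broadcasts a FoldingWings delegate.
# Chained additive Modify Bone nodes layer those states onto the wings and pack bone, then cache the pose.
# AnimDynamics springs run on that cached pose, with lower alpha while tactical sprinting or crouch-moving.
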AnimBlueprint /BRCosmetics/Accessories/FORT_Backpacks/Backpack_M_MED_BikeMold/Meshes/M_MED_BikeMold_Pack_AnimBP
target_skeleton=&/M_MED_BikeMold_Pack_Skeleton
graphs [
 AnimGraph {nodes [
   n1 = Get DynAl() -> (DynAl:double)
   n2 = Anim Dynamics - Bone: dyn_upper_l_jnt(ComponentPose=Local To Component(LocalPose=Use cached pose 'BaseMovState'()), Alpha=@n1, ExternalForce="0,0,0") {
    alpha=0.5
    angularDampingOverride=1
    angularSpringConstant=10000
    bAngularSpring=true
    bGravityOverrideInSimSpace=true
    bOverrideAngularDamping=true
    bUseGravityOverride=true
    simulationSpace=RootRelative
    alphaBoolBlend {blendOption=Linear}
    alphaScaleBiasClamp {interpSpeedDecreasing=50, interpSpeedIncreasing=50}
    boundBone {boneName=dyn_upper_l_jnt}
    componentAppliedLinearAccClamp=<1100 1100 1100>
    componentLinearAccScale=<1 15 1>
    componentLinearVelScale=<0.5 0.5 0.5>
    gravityOverride {z=-980}
    physicsBodyDefinitions [{
      sphereCollisionRadius=0
      boundBone {boneName=dyn_upper_l_jnt}
      boxExtents=<10 20 20>
      constraintSetup {angularTargetAxis=AxisZ, angularLimitsMax=<5 25 0>, angularLimitsMin=<-5 -25 0>, angularTarget=<0 0 1>}
      localJointOffset=<1 0 0>
     }]
   }
   n3 = Anim Dynamics - Bone: dyn_upper_r_jnt(ComponentPose=@n2, Alpha=@n1, ExternalForce="0,0,0") {
    alpha=0.5
    angularDampingOverride=1
    angularSpringConstant=10000
    bAngularSpring=true
    bGravityOverrideInSimSpace=true
    bOverrideAngularDamping=true
    bUseGravityOverride=true
    simulationSpace=RootRelative
    alphaBoolBlend {blendOption=Linear}
    alphaScaleBiasClamp {interpSpeedDecreasing=50, interpSpeedIncreasing=50}
    boundBone {boneName=dyn_upper_r_jnt}
    componentAppliedLinearAccClamp=<1100 1100 1100>
    componentLinearAccScale=<1 15 1>
    componentLinearVelScale=<0.5 0.5 0.5>
    gravityOverride {z=-980}
    physicsBodyDefinitions [{
      sphereCollisionRadius=0
      boundBone {boneName=dyn_upper_r_jnt}
      boxExtents=<10 20 20>
      constraintSetup {angularTargetAxis=AxisZ, angularLimitsMax=<5 25 0>, angularLimitsMin=<-5 -25 0>, angularTarget=<0 0 1>}
      localJointOffset=<1 0 0>
     }]
   }
   n4 = Anim Dynamics - Bone: dyn_lower_l_jnt(ComponentPose=@n3, Alpha=@n1, ExternalForce="0,0,0") {
    alpha=0.5
    angularDampingOverride=1
    angularSpringConstant=10000
    bAngularSpring=true
    bGravityOverrideInSimSpace=true
    bOverrideAngularDamping=true
    bUseGravityOverride=true
    simulationSpace=RootRelative
    alphaBoolBlend {blendOption=Linear}
    alphaScaleBiasClamp {interpSpeedDecreasing=50, interpSpeedIncreasing=50}
    boundBone {boneName=dyn_lower_l_jnt}
    componentAppliedLinearAccClamp=<1100 1100 1100>
    componentLinearAccScale=<1 15 1>
    componentLinearVelScale=<0.5 0.5 0.5>
    gravityOverride {z=-980}
    physicsBodyDefinitions [{
      sphereCollisionRadius=0
      boundBone {boneName=dyn_lower_l_jnt}
      boxExtents=<10 20 20>
      constraintSetup {angularTargetAxis=AxisZ, angularLimitsMax=<5 25 0>, angularLimitsMin=<-5 -25 0>, angularTarget=<0 0 1>}
      localJointOffset=<1 0 0>
     }]
   }
   n5 = Anim Dynamics - Bone: dyn_lower_r_jnt(ComponentPose=@n4, Alpha=@n1, ExternalForce="0,0,0") {
    alpha=0.5
    angularDampingOverride=1
    angularSpringConstant=10000
    bAngularSpring=true
    bGravityOverrideInSimSpace=true
    bOverrideAngularDamping=true
    bUseGravityOverride=true
    simulationSpace=RootRelative
    alphaBoolBlend {blendOption=Linear}
    alphaScaleBiasClamp {interpSpeedDecreasing=50, interpSpeedIncreasing=50}
    boundBone {boneName=dyn_lower_r_jnt}
    componentAppliedLinearAccClamp=<1100 1100 1100>
    componentLinearAccScale=<1 15 1>
    componentLinearVelScale=<0.5 0.5 0.5>
    gravityOverride {z=-980}
    physicsBodyDefinitions [{
      sphereCollisionRadius=0
      boundBone {boneName=dyn_lower_r_jnt}
      boxExtents=<10 20 20>
      constraintSetup {angularTargetAxis=AxisZ, angularLimitsMax=<5 25 0>, angularLimitsMin=<-5 -25 0>, angularTarget=<0 0 1>}
      localJointOffset=<1 0 0>
     }]
   }
   Output Pose(Result=Component To Local(ComponentPose=@n5)) {layerGroup=None}
   n6 = Get Is SprintingForward() -> (Is SprintingForward:bool)
   n7 = "Transform (Modify) Bone - Bone: dyn_upper_l_jnt"!(
    ComponentPose=Local To Component(LocalPose=Input Pose())
    bAlphaBoolEnabled:bool=@n6
    Translation="0,0,0"
    Rotation="17.703,26.269,-28.874"
    Scale="1,1,1"
   ) -> (Pose) {
    alphaInputType=Bool
    rotationMode=BMM_Additive
    alphaBoolBlend {blendInTime=0.2, blendOption=Linear, blendOutTime=0.2}
    alphaScaleBiasClamp {bInterpResult=true, interpSpeedDecreasing=15, interpSpeedIncreasing=15}
    boneToModify {boneName=dyn_upper_l_jnt}
    rotation=<rot 17.703 26.269 -28.874>
   }
   n8 = "Transform (Modify) Bone - Bone: dyn_upper_r_jnt"!(ComponentPose=@n7, bAlphaBoolEnabled:bool=@n6, Translation="-0.359,-0.527,-0.428", Rotation="-19.416,-29.895,-24.238", Scale="1,1,1") -> (Pose) {
    alphaInputType=Bool
    rotationMode=BMM_Additive
    translationMode=BMM_Additive
    alphaBoolBlend {blendInTime=0.2, blendOption=Linear, blendOutTime=0.2}
    alphaScaleBiasClamp {bInterpResult=true, interpSpeedDecreasing=15, interpSpeedIncreasing=15}
    boneToModify {boneName=dyn_upper_r_jnt}
    rotation=<rot -19.416 -29.895 -24.238>
    translation=<-0.3592 -0.5274 -0.4284>
   }
   n9 = Get bFolding_wings() -> (bFolding_wings:bool)
   n10 = "Transform (Modify) Bone - Bone: dyn_upper_l_jnt"(ComponentPose=@n8, bAlphaBoolEnabled=@n9, Translation="-0.503,-1.127,-0.373", Rotation="12.780,33.848,-25.710", Scale="1,1,1") {
    alphaInputType=Bool
    rotationMode=BMM_Additive
    translationMode=BMM_Additive
    alphaBoolBlend {blendInTime=0.2, blendOption=Linear, blendOutTime=0.6}
    alphaScaleBias {scale=10}
    alphaScaleBiasClamp {bInterpResult=true, interpSpeedDecreasing=15, interpSpeedIncreasing=15}
    boneToModify {boneName=dyn_upper_l_jnt}
    rotation=<rot 12.78 33.848 -25.71>
    translation=<-0.503 -1.127 -0.373>
   }
   n11 = "Transform (Modify) Bone - Bone: dyn_upper_r_jnt"(ComponentPose=@n10, bAlphaBoolEnabled=@n9, Translation="0.503,-1.127,-0.373", Rotation="-12.186,-33.220,-24.328", Scale="1,1,1") {
    alphaInputType=Bool
    rotationMode=BMM_Additive
    translationMode=BMM_Additive
    alphaBoolBlend {blendInTime=0.2, blendOption=Linear, blendOutTime=0.6}
    alphaScaleBias {scale=10}
    alphaScaleBiasClamp {bInterpResult=true, interpSpeedDecreasing=15, interpSpeedIncreasing=15}
    boneToModify {boneName=dyn_upper_r_jnt}
    rotation=<rot -12.186 -33.22 -24.328>
    translation=<0.503 -1.127 -0.373>
   }
   n12 = "Transform (Modify) Bone - Bone: dyn_lower_l_jnt"!(ComponentPose=@n11, bAlphaBoolEnabled:bool=@n6, Translation="0,0,0", Rotation="1.813,35,4.618", Scale="1,1,1") -> (Pose) {
    alphaInputType=Bool
    rotationMode=BMM_Additive
    alphaBoolBlend {blendInTime=0.2, blendOption=Linear, blendOutTime=0.2}
    alphaScaleBiasClamp {bInterpResult=true, interpSpeedDecreasing=15, interpSpeedIncreasing=15}
    boneToModify {boneName=dyn_lower_l_jnt}
    rotation=<rot 1.813 35 4.618>
   }
   n13 = "Transform (Modify) Bone - Bone: dyn_lower_r_jnt"!(ComponentPose=@n12, bAlphaBoolEnabled:bool=@n6, Translation="0,0,0", Rotation="1.813,-35,4.618", Scale="1,1,1") -> (Pose) {
    alphaInputType=Bool
    rotationMode=BMM_Additive
    alphaBoolBlend {blendInTime=0.2, blendOption=Linear, blendOutTime=0.2}
    alphaScaleBiasClamp {bInterpResult=true, interpSpeedDecreasing=15, interpSpeedIncreasing=15}
    boneToModify {boneName=dyn_lower_r_jnt}
    rotation=<rot 1.813 -35 4.618>
   }
   n14 = "Transform (Modify) Bone - Bone: dyn_lower_l_jnt"(ComponentPose=@n13, bAlphaBoolEnabled=@n9, Translation="-1,0,0", Rotation="27.455,19.932,12.536", Scale="1,1,1") {
    alphaInputType=Bool
    rotationMode=BMM_Additive
    translationMode=BMM_Additive
    alphaBoolBlend {blendInTime=0.2, blendOption=Linear, blendOutTime=0.6}
    alphaScaleBias {scale=10}
    alphaScaleBiasClamp {bInterpResult=true, interpSpeedDecreasing=15, interpSpeedIncreasing=15}
    boneToModify {boneName=dyn_lower_l_jnt}
    rotation=<rot 27.455 19.932 12.536>
    translation {x=-1}
   }
   n15 = "Transform (Modify) Bone - Bone: dyn_lower_r_jnt"(ComponentPose=@n14, bAlphaBoolEnabled=@n9, Translation="1,0,0", Rotation="-27.007,-22.587,8.883", Scale="1,1,1") {
    alphaInputType=Bool
    rotationMode=BMM_Additive
    translationMode=BMM_Additive
    alphaBoolBlend {blendInTime=0.2, blendOption=Linear, blendOutTime=0.6}
    alphaScaleBias {scale=10}
    alphaScaleBiasClamp {bInterpResult=true, interpSpeedDecreasing=15, interpSpeedIncreasing=15}
    boneToModify {boneName=dyn_lower_r_jnt}
    rotation=<rot -27.007 -22.587 8.883>
    translation {x=1}
   }
   n16 = Get bIsCrouching()
   n17 = "Transform (Modify) Bone - Bone: dyn_upper_l_jnt"(ComponentPose=@n15, bAlphaBoolEnabled=@n16, Translation="-0.503,-1.127,-0.373", Rotation="12.780,33.848,-25.710", Scale="1,1,1") {
    alphaInputType=Bool
    rotationMode=BMM_Additive
    translationMode=BMM_Additive
    alphaBoolBlend {blendInTime=0.12, blendOption=Linear, blendOutTime=0.25}
    alphaScaleBias {scale=10}
    alphaScaleBiasClamp {bInterpResult=true, interpSpeedDecreasing=15, interpSpeedIncreasing=15}
    boneToModify {boneName=dyn_upper_l_jnt}
    rotation=<rot 12.78 33.848 -25.71>
    translation=<-0.503 -1.127 -0.373>
   }
   n18 = "Transform (Modify) Bone - Bone: dyn_upper_r_jnt"(ComponentPose=@n17, bAlphaBoolEnabled=@n16, Translation="0.503,-1.127,-0.373", Rotation="-12.186,-33.220,-24.328", Scale="1,1,1") {
    alphaInputType=Bool
    rotationMode=BMM_Additive
    translationMode=BMM_Additive
    alphaBoolBlend {blendInTime=0.12, blendOption=Linear, blendOutTime=0.25}
    alphaScaleBias {scale=10}
    alphaScaleBiasClamp {bInterpResult=true, interpSpeedDecreasing=15, interpSpeedIncreasing=15}
    boneToModify {boneName=dyn_upper_r_jnt}
    rotation=<rot -12.186 -33.22 -24.328>
    translation=<0.503 -1.127 -0.373>
   }
   n19 = "Transform (Modify) Bone - Bone: dyn_lower_l_jnt"(ComponentPose=@n18, bAlphaBoolEnabled=@n16, Translation="-1,0,0", Rotation="14.096,22.523,7.472", Scale="1,1,1") {
    alphaInputType=Bool
    rotationMode=BMM_Additive
    translationMode=BMM_Additive
    alphaBoolBlend {blendInTime=0.12, blendOption=Linear, blendOutTime=0.25}
    alphaScaleBias {scale=10}
    alphaScaleBiasClamp {bInterpResult=true, interpSpeedDecreasing=15, interpSpeedIncreasing=15}
    boneToModify {boneName=dyn_lower_l_jnt}
    rotation=<rot 14.096 22.523 7.472>
    translation {x=-1}
   }
   n20 = "Transform (Modify) Bone - Bone: dyn_lower_r_jnt"(ComponentPose=@n19, bAlphaBoolEnabled=@n16, Translation="1,0,0", Rotation="-11.562,-20.157,4.745", Scale="1,1,1") {
    alphaInputType=Bool
    rotationMode=BMM_Additive
    translationMode=BMM_Additive
    alphaBoolBlend {blendInTime=0.12, blendOption=Linear, blendOutTime=0.25}
    alphaScaleBias {scale=10}
    alphaScaleBiasClamp {bInterpResult=true, interpSpeedDecreasing=15, interpSpeedIncreasing=15}
    boneToModify {boneName=dyn_lower_r_jnt}
    rotation=<rot -11.562 -20.157 4.745>
    translation {x=1}
   }
   n21 = Get Y MovementNegative() -> (Y MovementNegative:double)
   n22 = "Transform (Modify) Bone - Bone: dyn_upper_l_jnt"!(ComponentPose=@n20, Alpha:float=@n21, Translation="0,0,0", Rotation="-2.766,-18.659,16.946", Scale="1,1,1") -> (Pose) {
    rotationMode=BMM_Additive
    alphaBoolBlend {blendInTime=0.15, blendOption=Linear, blendOutTime=0.25}
    alphaScaleBias {scale=1.25}
    alphaScaleBiasClamp {bInterpResult=true, interpSpeedDecreasing=15, interpSpeedIncreasing=15}
    boneToModify {boneName=dyn_upper_l_jnt}
    rotation=<rot -2.766 -18.659 16.946>
   }
   n23 = "Transform (Modify) Bone - Bone: dyn_upper_r_jnt"!(ComponentPose=@n22, Alpha:float=@n21, Translation="0,0,0", Rotation="2.766,18.659,16.946", Scale="1,1,1") -> (Pose) {
    rotationMode=BMM_Additive
    alphaBoolBlend {blendInTime=0.15, blendOption=Linear, blendOutTime=0.25}
    alphaScaleBias {scale=1.25}
    alphaScaleBiasClamp {bInterpResult=true, interpSpeedDecreasing=15, interpSpeedIncreasing=15}
    boneToModify {boneName=dyn_upper_r_jnt}
    rotation=<rot 2.766 18.659 16.946>
   }
   n24 = "Transform (Modify) Bone - Bone: dyn_lower_l_jnt"!(ComponentPose=@n23, Alpha:float=@n21, Translation="0,0,0", Rotation="-5.535,-26.785,8.623", Scale="1,1,1") -> (Pose) {
    rotationMode=BMM_Additive
    alphaBoolBlend {blendInTime=0.15, blendOption=Linear, blendOutTime=0.25}
    alphaScaleBias {scale=1.25}
    alphaScaleBiasClamp {bInterpResult=true, interpSpeedDecreasing=15, interpSpeedIncreasing=15}
    boneToModify {boneName=dyn_lower_l_jnt}
    rotation=<rot -5.535 -26.785 8.623>
   }
   n25 = "Transform (Modify) Bone - Bone: dyn_lower_r_jnt"!(ComponentPose=@n24, Alpha:float=@n21, Translation="0,0,0", Rotation="5.535,26.785,8.623", Scale="1,1,1") -> (Pose) {
    rotationMode=BMM_Additive
    alphaBoolBlend {blendInTime=0.15, blendOption=Linear, blendOutTime=0.25}
    alphaScaleBias {scale=1.25}
    alphaScaleBiasClamp {bInterpResult=true, interpSpeedDecreasing=15, interpSpeedIncreasing=15}
    boneToModify {boneName=dyn_lower_r_jnt}
    rotation=<rot 5.535 26.785 8.623>
   }
   n26 = Get bCondStopMoving() -> (bCondStopMoving:bool)
   n27 = "Transform (Modify) Bone - Bone: dyn_upper_l_jnt"!(ComponentPose=@n25, bAlphaBoolEnabled:bool=@n26, Translation="0,0,0", Rotation="-2.539,-23.971,21.780", Scale="1,1,1") -> (Pose) {
    alphaInputType=Bool
    rotationMode=BMM_Additive
    alphaBoolBlend {blendInTime=0.3, blendOption=Linear, blendOutTime=0.5}
    alphaScaleBias {scale=1.25}
    alphaScaleBiasClamp {bInterpResult=true, interpSpeedDecreasing=15, interpSpeedIncreasing=15}
    boneToModify {boneName=dyn_upper_l_jnt}
    rotation=<rot -2.539 -23.971 21.78>
   }
   n28 = "Transform (Modify) Bone - Bone: dyn_upper_r_jnt"!(ComponentPose=@n27, bAlphaBoolEnabled:bool=@n26, Translation="0,0,0", Rotation="2.539,23.971,21.78", Scale="1,1,1") -> (Pose) {
    alphaInputType=Bool
    rotationMode=BMM_Additive
    alphaBoolBlend {blendInTime=0.3, blendOption=Linear, blendOutTime=0.5}
    alphaScaleBias {scale=1.25}
    alphaScaleBiasClamp {bInterpResult=true, interpSpeedDecreasing=15, interpSpeedIncreasing=15}
    boneToModify {boneName=dyn_upper_r_jnt}
    rotation=<rot 2.539 23.971 21.78>
   }
   n29 = "Transform (Modify) Bone - Bone: dyn_lower_l_jnt"!(ComponentPose=@n28, bAlphaBoolEnabled:bool=@n26, Translation="0,0,0", Rotation="3.690,-39.554,2.578", Scale="1,1,1") -> (Pose) {
    alphaInputType=Bool
    rotationMode=BMM_Additive
    alphaBoolBlend {blendInTime=0.3, blendOption=Linear, blendOutTime=0.5}
    alphaScaleBias {scale=1.25}
    alphaScaleBiasClamp {bInterpResult=true, interpSpeedDecreasing=15, interpSpeedIncreasing=15}
    boneToModify {boneName=dyn_lower_l_jnt}
    rotation=<rot 3.69 -39.554 2.578>
   }
   n30 = "Transform (Modify) Bone - Bone: dyn_lower_r_jnt"!(ComponentPose=@n29, bAlphaBoolEnabled:bool=@n26, Translation="0,0,0", Rotation="3.69,39.554,2.578", Scale="1,1,1") -> (Pose) {
    alphaInputType=Bool
    rotationMode=BMM_Additive
    alphaBoolBlend {blendInTime=0.3, blendOption=Linear, blendOutTime=0.5}
    alphaScaleBias {scale=1.25}
    alphaScaleBiasClamp {bInterpResult=true, interpSpeedDecreasing=15, interpSpeedIncreasing=15}
    boneToModify {boneName=dyn_lower_r_jnt}
    rotation=<rot 3.69 39.554 2.578>
   }
   n31 = Get PlayerVector_r() -> (PlayerVector_r:double)
   n32 = "Transform (Modify) Bone - Bone: dyn_upper_l_jnt"!(ComponentPose=@n30, Alpha:float=@n31, Translation="-0.809,-0.330,0.396", Rotation="3.903,13.844,-3.476", Scale="1,1,1") -> (Pose) {
    rotationMode=BMM_Additive
    translationMode=BMM_Additive
    alphaBoolBlend {blendOption=Linear}
    alphaScaleBiasClamp {bInterpResult=true, interpSpeedDecreasing=14, interpSpeedIncreasing=14}
    boneToModify {boneName=dyn_upper_l_jnt}
    rotation=<rot 3.903 13.844 -3.476>
    translation=<-0.809 -0.33 0.396>
   }
   n33 = "Transform (Modify) Bone - Bone: dyn_upper_r_jnt"!(ComponentPose=@n32, Alpha:float=@n31, Translation="0,0,0", Rotation="-15.810,-3.378,12.191", Scale="1,1,1") -> (Pose) {
    rotationMode=BMM_Additive
    alphaBoolBlend {blendOption=Linear}
    alphaScaleBiasClamp {bInterpResult=true, interpSpeedDecreasing=14, interpSpeedIncreasing=14}
    boneToModify {boneName=dyn_upper_r_jnt}
    rotation=<rot -15.81 -3.378 12.191>
   }
   n34 = Get PlayerVector_L() -> (PlayerVector_L:double)
   n35 = "Transform (Modify) Bone - Bone: dyn_upper_l_jnt"(ComponentPose=@n33, Alpha=@n34, Translation="-1.375,-0.726,-0.043", Rotation="5.247,-7.719,18.679", Scale="1,1,1") {
    rotationMode=BMM_Additive
    translationMode=BMM_Additive
    alphaBoolBlend {blendOption=Linear}
    alphaScaleBiasClamp {bInterpResult=true, interpSpeedDecreasing=14, interpSpeedIncreasing=14}
    boneToModify {boneName=dyn_upper_l_jnt}
    rotation=<rot 5.247 -7.719 18.679>
    translation=<-1.375 -0.726 -0.043>
   }
   n36 = "Transform (Modify) Bone - Bone: dyn_upper_r_jnt"(ComponentPose=@n35, Alpha=@n34, Translation="0,0,0", Rotation="0,0,-5", Scale="1,1,1") {
    rotationMode=BMM_Additive
    alphaBoolBlend {blendOption=Linear}
    alphaScaleBiasClamp {bInterpResult=true, interpSpeedDecreasing=14, interpSpeedIncreasing=14}
    boneToModify {boneName=dyn_upper_r_jnt}
    rotation {roll=-5}
   }
   n37 = "Transform (Modify) Bone - Bone: dyn_lower_l_jnt"!(ComponentPose=@n36, Alpha:float=@n31, Translation="0,0,0", Rotation="16.356,11.129,7.699", Scale="1,1,1") -> (Pose) {
    rotationMode=BMM_Additive
    alphaBoolBlend {blendOption=Linear}
    alphaScaleBiasClamp {bInterpResult=true, interpSpeedDecreasing=14, interpSpeedIncreasing=14}
    boneToModify {boneName=dyn_lower_l_jnt}
    rotation=<rot 16.356 11.129 7.699>
   }
   n38 = "Transform (Modify) Bone - Bone: dyn_lower_r_jnt"!(ComponentPose=@n37, Alpha:float=@n31, Translation="1.465,0.557,-0.736", Rotation="-5.031,32.750,19.453", Scale="1,1,1") -> (Pose) {
    rotationMode=BMM_Additive
    translationMode=BMM_Additive
    alphaBoolBlend {blendOption=Linear}
    alphaScaleBiasClamp {bInterpResult=true, interpSpeedDecreasing=14, interpSpeedIncreasing=14}
    boneToModify {boneName=dyn_lower_r_jnt}
    rotation=<rot -5.031 32.75 19.453>
    translation=<1.465 0.557 -0.736>
   }
   n39 = "Transform (Modify) Bone - Bone: dyn_lower_l_jnt"(ComponentPose=@n38, Alpha=@n34, Translation="-1.165,0.271,0.079", Rotation="-4.985,-9.906,1.916", Scale="1,1,1") {
    rotationMode=BMM_Additive
    translationMode=BMM_Additive
    alphaBoolBlend {blendOption=Linear}
    alphaScaleBiasClamp {bInterpResult=true, interpSpeedDecreasing=14, interpSpeedIncreasing=14}
    boneToModify {boneName=dyn_lower_l_jnt}
    rotation=<rot -4.985 -9.906 1.916>
    translation=<-1.165 0.271 0.079>
   }
   n40 = "Transform (Modify) Bone - Bone: dyn_lower_r_jnt"(ComponentPose=@n39, Alpha=@n34, Translation="0,0,0", Rotation="-12.281,-1.482,7.010", Scale="1,1,1") {
    rotationMode=BMM_Additive
    alphaBoolBlend {blendOption=Linear}
    alphaScaleBiasClamp {bInterpResult=true, interpSpeedDecreasing=14, interpSpeedIncreasing=14}
    boneToModify {boneName=dyn_lower_r_jnt}
    rotation=<rot -12.281 -1.482 7.01>
   }
   n41 = Get bPlayerVector_r_flight() -> (bPlayerVector_r_flight:bool)
   n42 = "Transform (Modify) Bone - Bone: dyn_upper_l_jnt"!(ComponentPose=@n40, bAlphaBoolEnabled:bool=@n41, Translation="0,0,0", Rotation="7.832,20.903,-9.028", Scale="1,1,1") -> (Pose) {
    alphaInputType=Bool
    rotationMode=BMM_Additive
    translationMode=BMM_Additive
    alphaBoolBlend {blendInTime=0.1, blendOption=Linear, blendOutTime=0.1}
    alphaScaleBiasClamp {bInterpResult=true, interpSpeedDecreasing=13, interpSpeedIncreasing=13}
    boneToModify {boneName=dyn_upper_l_jnt}
    rotation=<rot 7.832 20.903 -9.028>
   }
   n43 = Get bPlayerVector_l_flight() -> (bPlayerVector_l_flight:bool)
   n44 = "Transform (Modify) Bone - Bone: dyn_upper_r_jnt"!(ComponentPose=@n42, bAlphaBoolEnabled:bool=@n43, Translation="0,0,0", Rotation="-13.940,-35.033,-5.170", Scale="1,1,1") -> (Pose) {
    alphaInputType=Bool
    rotationMode=BMM_Additive
    alphaBoolBlend {blendInTime=0.1, blendOption=Linear, blendOutTime=0.1}
    alphaScaleBiasClamp {bInterpResult=true, interpSpeedDecreasing=13, interpSpeedIncreasing=13}
    boneToModify {boneName=dyn_upper_r_jnt}
    rotation=<rot -13.94 -35.033 -5.17>
   }
   n45 = "Transform (Modify) Bone - Bone: dyn_lower_l_jnt"!(ComponentPose=@n44, bAlphaBoolEnabled:bool=@n41, Translation="0,0,0", Rotation="16.356,11.129,7.699", Scale="1,1,1") -> (Pose) {
    alphaInputType=Bool
    rotationMode=BMM_Additive
    alphaBoolBlend {blendInTime=0.1, blendOption=Linear, blendOutTime=0.1}
    alphaScaleBiasClamp {bInterpResult=true, interpSpeedDecreasing=13, interpSpeedIncreasing=13}
    boneToModify {boneName=dyn_lower_l_jnt}
    rotation=<rot 16.356 11.129 7.699>
   }
   n46 = "Transform (Modify) Bone - Bone: dyn_lower_r_jnt"!(ComponentPose=@n45, bAlphaBoolEnabled:bool=@n43, Translation="0,0,0", Rotation="-12.281,-1.482,7.010", Scale="1,1,1") -> (Pose) {
    alphaInputType=Bool
    rotationMode=BMM_Additive
    alphaBoolBlend {blendInTime=0.1, blendOption=Linear, blendOutTime=0.1}
    alphaScaleBiasClamp {bInterpResult=true, interpSpeedDecreasing=13, interpSpeedIncreasing=13}
    boneToModify {boneName=dyn_lower_r_jnt}
    rotation=<rot -12.281 -1.482 7.01>
   }
   n47 = Get bCondJumping() -> (bCondJumping:bool)
   n48 = "Transform (Modify) Bone - Bone: dyn_upper_l_jnt"(ComponentPose=@n46, bAlphaBoolEnabled=@n47, Translation="0,0,0", Rotation="-2.766,-18.659,16.946", Scale="1,1,1") {
    alphaInputType=Bool
    rotationMode=BMM_Additive
    alphaBoolBlend {blendInTime=0.15, blendOption=Linear, blendOutTime=0.3}
    alphaScaleBiasClamp {bInterpResult=true, interpSpeedDecreasing=15, interpSpeedIncreasing=15}
    boneToModify {boneName=dyn_upper_l_jnt}
    rotation=<rot -2.766 -18.659 16.946>
   }
   n49 = "Transform (Modify) Bone - Bone: dyn_upper_r_jnt"(ComponentPose=@n48, bAlphaBoolEnabled=@n47, Translation="0,0,0", Rotation="2.766,18.659,16.946", Scale="1,1,1") {
    alphaInputType=Bool
    rotationMode=BMM_Additive
    alphaBoolBlend {blendInTime=0.15, blendOption=Linear, blendOutTime=0.3}
    alphaScaleBiasClamp {bInterpResult=true, interpSpeedDecreasing=15, interpSpeedIncreasing=15}
    boneToModify {boneName=dyn_upper_r_jnt}
    rotation=<rot 2.766 18.659 16.946>
   }
   n50 = "Transform (Modify) Bone - Bone: dyn_lower_l_jnt"(ComponentPose=@n49, bAlphaBoolEnabled=@n47, Translation="0,0,0", Rotation="-5.535,-26.785,8.623", Scale="1,1,1") {
    alphaInputType=Bool
    rotationMode=BMM_Additive
    alphaBoolBlend {blendInTime=0.15, blendOption=Linear, blendOutTime=0.3}
    alphaScaleBiasClamp {bInterpResult=true, interpSpeedDecreasing=15, interpSpeedIncreasing=15}
    boneToModify {boneName=dyn_lower_l_jnt}
    rotation=<rot -5.535 -26.785 8.623>
   }
   n51 = "Transform (Modify) Bone - Bone: dyn_lower_r_jnt"(ComponentPose=@n50, bAlphaBoolEnabled=@n47, Translation="0,0,0", Rotation="5.535,26.785,8.623", Scale="1,1,1") {
    alphaInputType=Bool
    rotationMode=BMM_Additive
    alphaBoolBlend {blendInTime=0.15, blendOption=Linear, blendOutTime=0.3}
    alphaScaleBiasClamp {bInterpResult=true, interpSpeedDecreasing=15, interpSpeedIncreasing=15}
    boneToModify {boneName=dyn_lower_r_jnt}
    rotation=<rot 5.535 26.785 8.623>
   }
   n52 = Get bGreaterVel() -> (bGreaterVel:bool)
   n53 = "Transform (Modify) Bone - Bone: dyn_upper_l_jnt"(ComponentPose=@n51, bAlphaBoolEnabled=@n52, Translation="0,0,0", Rotation="45.920,19.780,-45.920", Scale="1,1,1") {
    alphaInputType=Bool
    rotationMode=BMM_Additive
    alphaBoolBlend {blendInTime=0.1, blendOption=Linear, blendOutTime=0.5}
    alphaScaleBiasClamp {bInterpResult=true, interpSpeedDecreasing=15, interpSpeedIncreasing=15}
    boneToModify {boneName=dyn_upper_l_jnt}
    rotation=<rot 45.92 19.78 -45.92>
   }
   n54 = "Transform (Modify) Bone - Bone: dyn_upper_r_jnt"(ComponentPose=@n53, bAlphaBoolEnabled=@n52, Translation="0,0,0", Rotation="-59.874,-3.172,-61.371", Scale="1,1,1") {
    alphaInputType=Bool
    rotationMode=BMM_Additive
    alphaBoolBlend {blendInTime=0.1, blendOption=Linear, blendOutTime=0.5}
    alphaScaleBiasClamp {bInterpResult=true, interpSpeedDecreasing=15, interpSpeedIncreasing=15}
    boneToModify {boneName=dyn_upper_r_jnt}
    rotation=<rot -59.874 -3.172 -61.371>
   }
   n55 = "Transform (Modify) Bone - Bone: dyn_lower_l_jnt"(ComponentPose=@n54, bAlphaBoolEnabled=@n52, Translation="0,0,0", Rotation="17.750,34.984,2.627", Scale="1,1,1") {
    alphaInputType=Bool
    rotationMode=BMM_Additive
    alphaBoolBlend {blendInTime=0.1, blendOption=Linear, blendOutTime=0.5}
    alphaScaleBiasClamp {bInterpResult=true, interpSpeedDecreasing=15, interpSpeedIncreasing=15}
    boneToModify {boneName=dyn_lower_l_jnt}
    rotation=<rot 17.75 34.984 2.627>
   }
   n56 = "Transform (Modify) Bone - Bone: dyn_lower_r_jnt"(ComponentPose=@n55, bAlphaBoolEnabled=@n52, Translation="0,0,0", Rotation="-21.620,-39.917,-1.168", Scale="1,1,1") {
    alphaInputType=Bool
    rotationMode=BMM_Additive
    alphaBoolBlend {blendInTime=0.1, blendOption=Linear, blendOutTime=0.5}
    alphaScaleBiasClamp {bInterpResult=true, interpSpeedDecreasing=15, interpSpeedIncreasing=15}
    boneToModify {boneName=dyn_lower_r_jnt}
    rotation=<rot -21.62 -39.917 -1.168>
   }
   n57 = "Transform (Modify) Bone - Bone: pack"!(ComponentPose=@n56, bAlphaBoolEnabled:bool=Get bIsSkyDiving(), Translation="0,-2,0", Rotation="0,0,-10", Scale) -> (Pose) {
    alphaInputType=Bool
    rotationMode=BMM_Additive
    translationMode=BMM_Additive
    alphaBoolBlend {blendInTime=0.1, blendOption=Linear, blendOutTime=0.1}
    alphaScaleBiasClamp {bInterpResult=true, interpSpeedDecreasing=15, interpSpeedIncreasing=15}
    boneToModify {boneName=pack}
    rotation {roll=-10}
    translation {y=-2}
   }
   n58 = "Transform (Modify) Bone - Bone: pack"!(
    ComponentPose=@n57
    bAlphaBoolEnabled="Get Fly_condition "() -> ("Fly_condition ":bool)
    Translation="0,-2,0"
    Rotation_Roll:float=0
    Rotation_Pitch:float=0
    Rotation_Yaw:float=Get!() -> (RotateSpineCond_inSky_Roll:float, RotateSpineCond_inSky_Pitch:float, RotateSpineCond_inSky_Yaw:float).RotateSpineCond_inSky_Yaw
    Scale
   ) -> (Pose) {
    alphaInputType=Bool
    rotationMode=BMM_Additive
    translationMode=BMM_Additive
    alphaBoolBlend {blendInTime=0.1, blendOption=Linear, blendOutTime=0.1}
    alphaScaleBiasClamp {bInterpResult=true, interpSpeedDecreasing=15, interpSpeedIncreasing=15}
    boneToModify {boneName=pack}
    translation {y=-2}
   }
   n59 = "Transform (Modify) Bone - Bone: pack"!(ComponentPose=@n58, bAlphaBoolEnabled=NOT Boolean(A=Get bIsSkyDiving()), Translation="0,0.7,0", Rotation:Rotator="0,0,0", Scale) -> (Pose) {
    alphaInputType=Bool
    rotationMode=BMM_Additive
    translationMode=BMM_Additive
    alphaBoolBlend {blendInTime=0.1, blendOption=Linear, blendOutTime=0.1}
    alphaScaleBiasClamp {bInterpResult=true, interpSpeedDecreasing=15, interpSpeedIncreasing=15}
    boneToModify {boneName=pack}
    translation {y=0.7}
   }
   Save cached pose 'BaseMovState'(Pose=Component To Local(ComponentPose=@n59)) {cacheName=BaseMovState}
  ]}
 EventGraph {nodes [
   Event BlueprintUpdateAnimation()
   > n1 = Sequence!(execute) -> (then_0, then_1, then_2:Exec, then_3:Exec, then_4:Exec, then_5:Exec)
   n2 = Get bIsJumping()
   n3 = OR Boolean(A=OR Boolean(A=Get bIsTacticalSprinting(), B=Get bIsTargeting()), B=Get bIsDBNO())
   n4 = Get bIsSkydiveDiveMode()
   > n5 = Set bFolding_wings(bFolding_wings:bool=AND Boolean(A=OR Boolean(A=@n3, B=@n4), B=NOT Boolean(A=@n2))) -> (Output_Get:bool)
   > Call Folding Wings!(execute:Exec, self:M MED Bike Mold Pack Anim BP Object Reference, bIsFoldingWings:bool=@n5) -> (then:Exec) {is=title K2Node_CallDelegate}
   n6 = Get bIsAccelerating2D()
   n7 = Get bIsCrouching()
   n8 = Branch(execute=@n1.then_2, Condition=OR Boolean(A=Get bIsTacticalSprinting(), B=AND Boolean(A=@n6, B=@n7)))
   > Set DynAl(DynAl:double=0.3) -> (Output_Get:double)
   Set DynAl(execute=@n8.else, DynAl=0.5)
   n9 = "float > float"(A="Max (Float)"(A=Get PawnVelocityZ(), B=0), B=3000)
   n10 = "Not Equal (Boolean)"(A=Get bIsParachuteOpened(), B=true) -> (ReturnValue=~)
   n11 = Branch(execute=@n1.then_3, Condition=AND Boolean(A=AND Boolean(A=@n9, B=@n10), B=Get bIsSkyDiving()))
   > Set bGreaterVel(bGreaterVel:bool=true) -> (Output_Get:bool)
   Set bGreaterVel(execute=@n11.else, bGreaterVel=false)
   n12 = Branch(execute=@n1.then_1, Condition=OR Boolean(A=@n2, B=Get bIsFalling()))
   > Set Is SprintingForward(Is SprintingForward:bool=false) -> (Output_Get:bool)
   n13 = Get bIsCrouching()
   Set Is SprintingForward(execute=@n12.else, Is SprintingForward=AND Boolean(A=Get bIsSprinting(), B=NOT Boolean(A=@n13)))
   "Set Fly_condition "(execute=@n1.then_4, "Fly_condition ":bool=Get bIsSkydiveFloating()) -> (Output_Get:bool)
   n14 = GetAlphaFromReferenceTransformDelta(
    SkeletalMeshComponent=GetOwnerSkeletalMeshComponent()
    BoneName=spine_04
    BaseName=spine_03
    DeltaComponentType=EulerX
    InputRange=(X=0,Y=1)
    TwistAxis="0,0,1"
   )
   n15 = BreakRotator(InRot=Invert Rotator!(A:Rotator=MakeRotator(Yaw=@n14.RawOutput)) -> (ReturnValue:Rotator))
   > Set RotateSpineCond_inSky(RotateSpineCond_inSky:Rotator=MakeRotator(Yaw=@n15.Yaw)) -> (Output_Get:Rotator)
   n16 = Sequence!(execute=@n1.then_5) -> (then_0, then_1, then_2, then_3)
   > n17 = PlayerMovement(AnimInstance=Self-Reference(), XDotProductCoefficient=1, YDotProductCoefficient=1, ZDotProductCoefficient=1)
   n18 = Round(A="Min (Float)"(A:double=@n17.YMovement, B:double=0) -> (ReturnValue:double) {is=title K2Node_CommutativeAssociativeBinaryOperator}) {is=title K2Node_CallFunction}
   > Set Y MovementNegative(Y MovementNegative:double="Absolute (Float)"(A="To Float (Integer)"(InInt=@n18))) -> (Output_Get:double)
   > n19 = Set PlayerVector_r(PlayerVector_r:double="Max (Float)"(A=@n17, B=0)) -> (Output_Get:double)
   > n20 = Set PlayerVector_L(PlayerVector_L:double="Absolute (Float)"(A="Min (Float)"(A=@n17, B=0))) -> (Output_Get:double)
   > n21 = Branch(Condition=Get bIsSprinting())
   > n22 = Set bCondStopMoving(bCondStopMoving:bool=false) -> (Output_Get:bool)
   Do Once(execute=@n21.else, Reset=@n22.then)
   n23 = Get bIsCrouching()
   > n24 = Branch(Condition=AND Boolean(A=NOT Boolean(A=@n23), B=true))
   > Set bCondStopMoving(bCondStopMoving=true)
   > Delay(Duration=0.2)
   > Set bCondStopMoving(bCondStopMoving=false)
   Set bCondStopMoving(execute=@n24.else, bCondStopMoving=false)
   n25 = Get bIsSkyDiving()
   Set bCondJumping(execute=@n16.then_3, bCondJumping:bool=AND Boolean(A=Get bIsJumping(), B=NOT Boolean(A=@n25))) -> (Output_Get:bool)
   n26 = Round(A=@n20)
   n27 = AND Boolean(A=Get bIsAccelerating2D(), B=Get bIsSkyDiving())
   Set bPlayerVector_l_flight(
    execute=[@n29.then, @n16.then_2]
    bPlayerVector_l_flight:bool=AND Boolean(A=@n27, B="To Boolean (Integer)"!(InInt:int=@n26) -> (ReturnValue:bool))
   ) -> (Output_Get:bool)
   n28 = Round(A=@n19)
   n29 = Set bPlayerVector_r_flight(execute=@n16.then_1, bPlayerVector_r_flight:bool=AND Boolean(A="To Boolean (Integer)"(InInt=@n28), B=@n27)) -> (Output_Get:bool)
   Event OnBeginSkydiving()
   > Set DynAl(DynAl=0.3)
   Event OnEndSkydiving()
   > Set DynAl(DynAl=0.5)
   float * float(A=@n15.Yaw, B=1.5) {is=title K2Node_PromotableOperator}
   Set PlayerVector_FwdBwd(PlayerVector_FwdBwd:double=0.1) -> (Output_Get:double)
   n30 = TryGetPawnOwner()
   n31 = Vector Length(A=MakeVector(Y=GetVelocity!(self:Actor Object Reference=@n30) -> (ReturnValue_X:double, ReturnValue_Y:double, ReturnValue_Z:double).ReturnValue_Y))
   Set PlayerVector_FwdBwd(PlayerVector_FwdBwd=MapRangeClamped(Value=FInterpTo(Current=0, Target=@n31, DeltaTime=1, InterpSpeed=0.1), InRangeA=0, InRangeB=45))
  ]}
 FoldingWings {nodes [FoldingWings!() -> (then, bIsFoldingWings:bool=~) {is=title K2Node_FunctionEntry}]}
]
variables [
 bCondJumping {type=boolean}
 bCondStopMoving {type=boolean}
 bFolding_wings {type=boolean}
 bGreaterVel {type=boolean}
 bPlayerVector_l_flight {type=boolean}
 bPlayerVector_r_flight {type=boolean}
 DynAl {type=number}
 "Fly_condition " {type=boolean}
 FoldingWings {default=()}
 ifTacticalSprint {type=number}
 Is SprintingForward {type=boolean}
 NewVar_0 {type=PrimitiveComponent, default=null}
 PlayerVector_FwdBwd {type=number}
 PlayerVector_L {type=number}
 PlayerVector_r {type=number}
 RotateSpineCond_inSky {type=Rotator}
 Y MovementNegative {type=number}
]
